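# Physics asset for the head part of the female medium Street Racer White character, set up around ponytail physics.
# Capsule bodies cover upper spine, left clavicle and upper arm, neck and head, forming collision volumes around the hair.
# A base ponytail body links the head to a five-bone dynamic chain.
# Torso, neck and ponytail-root joints swing and twist freely, while lower ponytail joints use tighter, individually tuned angular limits.
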PhysicsAsset /BRCosmetics/Characters/Player/Female/Medium/Bodies/F_MED_Street_Racer_White_01/Meshes/Parts/F_MED_Street_Racer_White_Head_Physics
bodies {
 base_ponytail_1 [{shapeType=Capsule, length=0.505, radius=0.505, shapeName=base_ponytail_1_capsule, rotation=<rot 0 -90 0>}]
 clavicle_l [
  {
   shapeType=Capsule
   length=7.6907
   radius=4.3767
   shapeName=clavicle_l_capsule
   center=<4.1035 2.5489 -1.6038>
   rotation=<rot 1.0055 -99.5247 96.0289>
  }
  {
   shapeType=Capsule
   length=5.5
   radius=5.1913
   shapeName=clavicle_l_capsule
   center=<6.2091 -1.4299 0.2052>
   rotation=<rot -0.0005 -90 103.1663>
  }
 ]
 dyn_ponytail_1 [{
   shapeType=Capsule
   length=2.0236
   radius=2.5
   shapeName=dyn_ponytail_1_capsule
   center=<2.7226 -0.4853 0.3708>
   rotation=<rot 40.6402 -89.5507 93.4348>
  }]
 dyn_ponytail_2 [{
   shapeType=Capsule
   length=1.7486
   radius=0.8585
   shapeName=dyn_ponytail_2_capsule
   center=<1.4775 -0.2353 0.1325>
   rotation=<rot -0.3811 80.4278 -87.8381>
  }]
 dyn_ponytail_3 [{
   shapeType=Capsule
   length=4.6359
   radius=2.0798
   shapeName=dyn_ponytail_3_capsule
   center=<2.3291 -1.0153 0.7261>
   rotation=<rot 4.1287 -93.6857 -94.1173>
  }]
 dyn_ponytail_4 [
  {
   shapeType=Capsule
   length=3.2556
   radius=3.0576
   shapeName=dyn_ponytail_4_capsule
   center=<1.8561 0.7284 0.6927>
   rotation=<rot 3.8852 -74.3072 -99.7938>
  }
  {
   shapeType=Capsule
   length=1.477
   radius=2.3636
   shapeName=dyn_ponytail_4_capsule
   center=<1.103 0.7673 1.6914>
   rotation=<rot 0.8937 -84.8319 -67.4848>
  }
 ]
 dyn_ponytail_5 [{shapeType=Capsule, length=1.7057, radius=1.955, shapeName=dyn_ponytail_5_capsule, center=<2.4137 0 -0.3609>, rotation=<rot -90 180 180>}]
 head [
  {
   shapeType=Capsule
   length=13.0091
   radius=10.6788
   shapeName=head_capsule
   center=<4.9644 2.9025 -0.1497>
   rotation=<rot -90 -3.0594 -40.8383>
  }
  {
   shapeType=Capsule
   length=6.5457
   radius=5.705
   shapeName=head_capsule
   center=<0.6042 3.1994 -0.1373>
   rotation=<rot -78.4066 179.9995 -180>
  }
 ]
 neck_01 [{shapeType=Capsule, length=1.8448, radius=5.1716, shapeName=neck_01_capsule, rotation=<rot 0 -90.0001 89.0985>}]
 neck_02 [{shapeType=Capsule, length=2.8461, radius=4.3899, shapeName=neck_02_capsule, center=<0 0.3389 0>, rotation=<rot 0 -90 89.648>}]
 spine_03 [{shapeType=Capsule, length=13.5, radius=5.5, shapeName=spine_03_capsule, rotation=<rot 0 -90 0>}]
 spine_04 [{shapeType=Capsule, length=13.5, radius=5.5, shapeName=spine_04_capsule, rotation=<rot 0 -90 0>}]
 spine_05 [
  {shapeType=Capsule, length=10, radius=9, shapeName=spine_05_capsule, center=<0.8119 -2.4091 0>, rotation=<rot 0 -90 0>}
  {
   shapeType=Capsule
   length=31.7788
   radius=15.2168
   shapeName=spine_05_capsule
   center=<13.6975 -20.9261 0.5153>
   rotation=<rot 10.0201 -90.0002 0>
  }
 ]
 upperarm_l [{
   shapeType=Capsule
   length=12.564
   radius=6.3796
   shapeName=upperarm_l_capsule
   center=<6.9309 -0.2614 -1.1709>
   rotation=<rot -0.0001 -89.9993 90.7077>
  }]
}
constraints [
 {
  bone1Name=spine_04
  bone2Name=spine_03
  swing1LimitDegrees=45
  swing1Motion=Free
  swing2LimitDegrees=45
  swing2Motion=Free
  twistLimitDegrees=45
  twistMotion=Free
 }
 {
  bone1Name=spine_05
  bone2Name=spine_04
  swing1LimitDegrees=45
  swing1Motion=Free
  swing2LimitDegrees=45
  swing2Motion=Free
  twistLimitDegrees=45
  twistMotion=Free
 }
 {
  bone1Name=clavicle_l
  bone2Name=spine_05
  swing1LimitDegrees=45
  swing1Motion=Free
  swing2LimitDegrees=45
  swing2Motion=Free
  twistLimitDegrees=45
  twistMotion=Free
 }
 {
  bone1Name=upperarm_l
  bone2Name=clavicle_l
  swing1LimitDegrees=45
  swing1Motion=Free
  swing2LimitDegrees=45
  swing2Motion=Free
  twistLimitDegrees=45
  twistMotion=Free
 }
 {
  bone1Name=neck_01
  bone2Name=spine_05
  swing1LimitDegrees=45
  swing1Motion=Free
  swing2LimitDegrees=45
  swing2Motion=Free
  twistLimitDegrees=45
  twistMotion=Free
 }
 {
  bone1Name=neck_02
  bone2Name=neck_01
  swing1LimitDegrees=45
  swing1Motion=Free
  swing2LimitDegrees=45
  swing2Motion=Free
  twistLimitDegrees=45
  twistMotion=Free
 }
 {
  bone1Name=head
  bone2Name=neck_02
  swing1LimitDegrees=45
  swing1Motion=Free
  swing2LimitDegrees=45
  swing2Motion=Free
  twistLimitDegrees=45
  twistMotion=Free
 }
 {bone1Name=dyn_ponytail_5, bone2Name=dyn_ponytail_4, swing1LimitDegrees=45, swing2LimitDegrees=30, twistLimitDegrees=30}
 {
  bone1Name=base_ponytail_1
  bone2Name=head
  swing1LimitDegrees=45
  swing1Motion=Free
  swing2LimitDegrees=45
  swing2Motion=Free
  twistLimitDegrees=45
  twistMotion=Free
 }
 {
  bone1Name=dyn_ponytail_1
  bone2Name=base_ponytail_1
  swing1LimitDegrees=45
  swing1Motion=Free
  swing2LimitDegrees=45
  swing2Motion=Free
  twistLimitDegrees=45
  twistMotion=Free
 }
 {bone1Name=dyn_ponytail_2, bone2Name=dyn_ponytail_1, swing1LimitDegrees=12, swing2LimitDegrees=12, twistLimitDegrees=0}
 {bone1Name=dyn_ponytail_3, bone2Name=dyn_ponytail_2, swing1LimitDegrees=30, swing2LimitDegrees=25, twistLimitDegrees=0}
 {bone1Name=dyn_ponytail_4, bone2Name=dyn_ponytail_3, swing1LimitDegrees=40, swing2LimitDegrees=24, twistLimitDegrees=10}
]
