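# Head-part AnimBP for the female Street Racer that copies the attached parent's pose and animates the ponytail.
# It offsets head_offset using scaled head rotation and head scale, and drives ponytail bones through weighted multi-bone constraints.
# RigidBody simulates hair in spine_05 space with gravity override and jump/fall settings, switching to a posed parachute simulation when hang-gliding.
# With rigid bodies disabled, the cached constraint-only pose is used instead.
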AnimBlueprint /BRCosmetics/Characters/Player/Female/Medium/Bodies/F_MED_Street_Racer_White_01/Meshes/Parts/F_MED_Street_Racer_White_Head_AnimBP
target_skeleton=&/F_MED_Street_Racer_White_Head_Skeleton
graphs [
 AnimGraph {nodes [
   "Transform (Modify) Bone - Bone: dyn_ponytail_1"!(ComponentPose, Alpha=1, Translation="0.3147,0.0676,-0.3145", Rotation="19.2343,1.2028,-30.5518") -> (Pose) {
    rotationMode=BMM_Additive
    rotationSpace=BCS_BoneSpace
    translationMode=BMM_Additive
    translationSpace=BCS_BoneSpace
    alphaBoolBlend {blendOption=Linear}
    boneToModify {boneName=dyn_ponytail_1}
    rotation=<rot 19.2343 1.2028 -30.5518>
    translation=<0.3147 0.0676 -0.3145>
   }
   "Transform (Modify) Bone - Bone: dyn_ponytail_3"!(ComponentPose, Alpha=1, Translation="0,0,-2", Rotation="-44.4879,-27.2856,-58.0453") -> (Pose) {
    rotationMode=BMM_Additive
    rotationSpace=BCS_BoneSpace
    translationMode=BMM_Additive
    translationSpace=BCS_BoneSpace
    alphaBoolBlend {blendOption=Linear}
    boneToModify {boneName=dyn_ponytail_3}
    rotation=<rot -44.4879 -27.2856 -58.0453>
    translation {z=-2}
   }
   n1 = "Transform (Modify) Bone - Bone: base_ponytail_1"!(ComponentPose=Local To Component(LocalPose=Use cached pose 'firstChainConstraint'()), Alpha=1, Translation="-2,0,0") -> (Pose) {
    rotationMode=BMM_Additive
    rotationSpace=BCS_BoneSpace
    translationMode=BMM_Additive
    translationSpace=BCS_BoneSpace
    alphaBoolBlend {blendOption=Linear}
    boneToModify {boneName=base_ponytail_1}
    translation {x=-2}
   }
   n2 = "Transform (Modify) Bone - Bone: base_ponytail_1"!(ComponentPose=@n1, Alpha=1, Rotation:Rotator="12.961,1.347,-34.139") -> (Pose) {
    rotationMode=BMM_Additive
    rotationSpace=BCS_BoneSpace
    scaleMode=BMM_Additive
    scaleSpace=BCS_BoneSpace
    translationMode=BMM_Additive
    translationSpace=BCS_BoneSpace
    alphaBoolBlend {blendOption=Linear}
    alphaScaleBiasClamp {bInterpResult=true}
    boneToModify {boneName=base_ponytail_1}
    rotation=<rot 12.961 1.347 -34.139>
   }
   n3 = RigidBody!(
    ComponentPose=@n2
    Alpha
    OverrideWorldGravity:Vector=Get GravityOverride()
    SimSpaceSettings:Sim Space Settings Structure=Get Sim Space Settings Output() -> (Sim Space Settings Output:Sim Space Settings Structure)
   ) -> (Pose) {
    bOverrideWorldGravity=true
    overridePhysicsAsset=&/F_MED_Street_Racer_White_Head_Physics
    simulationSpace=BaseBoneSpace
    alphaBoolBlend {blendOption=Linear}
    baseBoneRef {boneName=spine_05}
    componentAppliedLinearAccClamp=<3000 3000 3000>
    componentLinearAccScale=<0.3 0.4 0.3>
    componentLinearVelScale=<0.3 0.4 0.3>
    simSpaceSettings {maxAngularVelocity=3, maxLinearAcceleration=700, maxLinearVelocity=500, velocityScaleZ=0, worldAlpha=0.65}
   }
   n4 = Component To Local(ComponentPose=@n3)
   n5 = Get alphaHairRot() -> (alphaHairRot:double)
   n6 = "Transform (Modify) Bone - Bone: dyn_ponytail_1"!(ComponentPose=Local To Component(LocalPose=Use cached pose 'startPose'()), Alpha=@n5) -> (Pose) {
    rotationMode=BMM_Additive
    rotationSpace=BCS_BoneSpace
    translationMode=BMM_Additive
    translationSpace=BCS_BoneSpace
    alphaBoolBlend {blendOption=Linear}
    alphaScaleBiasClamp {bInterpResult=true}
    boneToModify {boneName=dyn_ponytail_1}
    rotation {pitch=28}
    translation {z=-1.4}
   }
   n7 = RigidBody!(ComponentPose=@n6, Alpha) -> (Pose) {
    bOverrideWorldGravity=true
    overridePhysicsAsset=&/F_MED_Street_Racer_White_Head_PhysicsParachute
    simulationSpace=BaseBoneSpace
    alphaBoolBlend {blendOption=Linear}
    baseBoneRef {boneName=spine_05}
    componentAppliedLinearAccClamp=<3000 3000 3000>
    componentLinearAccScale=<0.3 0.4 0.3>
    componentLinearVelScale=<0.3 0.4 0.3>
    overrideWorldGravity {z=-980}
    simSpaceSettings {maxAngularVelocity=3, maxLinearAcceleration=700, maxLinearVelocity=500, velocityScaleZ=0}
   }
   n8 = "Transform (Modify) Bone - Bone: dyn_ponytail_3"(ComponentPose=@n7, Alpha=@n5, Translation="0,0,0", Rotation="-67.031,28.761,-15.632") {
    rotationMode=BMM_Additive
    rotationSpace=BCS_BoneSpace
    translationMode=BMM_Additive
    translationSpace=BCS_BoneSpace
    alphaBoolBlend {blendOption=Linear}
    boneToModify {boneName=dyn_ponytail_3}
    rotation=<rot -67.031 28.761 -15.632>
   }
   n9 = "Transform (Modify) Bone - Bone: dyn_ponytail_1"!(ComponentPose=@n8, Alpha=@n5, Translation:Vector="0,0,0", Rotation:Rotator="14.413,12.502,14.398") -> (Pose) {
    rotationMode=BMM_Additive
    rotationSpace=BCS_BoneSpace
    translationMode=BMM_Additive
    translationSpace=BCS_BoneSpace
    alphaBoolBlend {blendOption=Linear}
    boneToModify {boneName=dyn_ponytail_1}
    rotation=<rot 14.413 12.502 14.398>
   }
   n10 = Component To Local(ComponentPose=@n9)
   n11 = Get isUsingHangGlidingInParachute() -> (isUsingHangGlidingInParachute:bool)
   n12 = Local To Component(LocalPose=Blend Poses by bool(
     BlendPose_0=Blend Poses by bool(BlendPose_0=@n10, BlendPose_1=@n4, BlendTime_0=0.4, BlendTime_1=0.4, bActiveValue=@n11) {blendType=QuadraticInOut}
     BlendPose_1=Use cached pose 'onlyConstraints'()
     bActiveValue=Get bRigidBodyEnabled()
    ) {blendType=Linear})
   n13 = Get Head Scale() -> (Head Scale:Vector)
   Output Pose(Result=Component To Local(ComponentPose="Transform (Modify) Bone - Bone: head_offset"!(ComponentPose=@n12, Alpha=1, Scale=@n13) -> (Pose) {scaleMode=BMM_Additive, scaleSpace=BCS_BoneSpace, boneToModify {boneName=head_offset}})) {layerGroup=None}
   n14 = "Transform (Modify) Bone - Bone: head_offset"!(
    ComponentPose=Local To Component(LocalPose=Copy Pose From Mesh() {bUseAttachedParent=true})
    Alpha=1
    Rotation:Rotator=Get head rot() -> (head rot:Rotator)
   ) -> (Pose) {
    rotationMode=BMM_Replace
    rotationSpace=BCS_BoneSpace
    alphaBoolBlend {blendOption=Linear}
    alphaScaleBiasClamp {interpSpeedDecreasing=5, interpSpeedIncreasing=20}
    boneToModify {boneName=head_offset}
   }
   Save cached pose 'startPose'(Pose=Component To Local(ComponentPose=@n14)) {cacheName=startPose}
   n15 = "Constraint - dyn_ponytail_2 (PPPP)"!(
    ComponentPose=Local To Component(LocalPose=Use cached pose 'firstChainConstraint'())
    Alpha=1
    ConstraintWeights_0=0.444
    ConstraintWeights_1:float=0.3334
    ConstraintWeights_2:float=0.187
    ConstraintWeights_3:float=0.0356
   ) -> (Pose) {
    alphaBoolBlend {blendOption=Linear}
    boneToModify {boneName=dyn_ponytail_2}
    constraintSetup [{targetBone {boneName=neck_01}}, {targetBone {boneName=neck_02}}, {targetBone {boneName=spine_05}}, {targetBone {boneName=head}}]
    constraintWeights [1,1,1,1]
   }
   n16 = "Constraint - dyn_ponytail_3 (PPPPPP)"!(
    ComponentPose=@n15
    Alpha=1
    ConstraintWeights_0=0.373
    ConstraintWeights_1:float=0.3102
    ConstraintWeights_2:float=0.2256
    ConstraintWeights_3:float=0.0613
    ConstraintWeights_4:float=0.0285
    ConstraintWeights_5:float=0.0014
   ) -> (Pose) {
    alphaBoolBlend {blendOption=Linear}
    boneToModify {boneName=dyn_ponytail_3}
    constraintSetup [
     {targetBone {boneName=clavicle_l}}
     {targetBone {boneName=spine_05}}
     {targetBone {boneName=neck_01}}
     {targetBone {boneName=neck_02}}
     {targetBone {boneName=clavicle_r}}
     {targetBone {boneName=spine_03}}
    ]
    constraintWeights [1,1,1,1,1,1]
   }
   n17 = "Constraint - dyn_ponytail_4 (PPPPPP)"!(
    ComponentPose=@n16
    Alpha=1
    ConstraintWeights_0=0.6305
    ConstraintWeights_1:float=0.271
    ConstraintWeights_2:float=0.0632
    ConstraintWeights_3:float=0.0215
    ConstraintWeights_4:float=0.0128
    ConstraintWeights_5:float=0.001
   ) -> (Pose) {
    alphaBoolBlend {blendOption=Linear}
    boneToModify {boneName=dyn_ponytail_4}
    constraintSetup [
     {targetBone {boneName=clavicle_l}}
     {targetBone {boneName=spine_05}}
     {targetBone {boneName=neck_01}}
     {targetBone {boneName=neck_02}}
     {targetBone {boneName=clavicle_r}}
     {targetBone {boneName=spine_04}}
    ]
    constraintWeights [1,1,1,1,1,1]
   }
   n18 = "Constraint - dyn_ponytail_5 (PPPPPP)"!(
    ComponentPose=@n17
    Alpha=1
    ConstraintWeights_0=0.5394
    ConstraintWeights_1:float=0.3071
    ConstraintWeights_2:float=0.0719
    ConstraintWeights_3:float=0.04
    ConstraintWeights_4:float=0.0272
    ConstraintWeights_5:float=0.0144
   ) -> (Pose) {
    alphaBoolBlend {blendOption=Linear}
    boneToModify {boneName=dyn_ponytail_5}
    constraintSetup [
     {targetBone {boneName=clavicle_l}}
     {targetBone {boneName=spine_05}}
     {targetBone {boneName=upperarm_l}}
     {targetBone {boneName=neck_01}}
     {targetBone {boneName=spine_04}}
     {targetBone {boneName=clavicle_r}}
    ]
    constraintWeights [1,1,1,1,1,1]
   }
   Save cached pose 'onlyConstraints'(Pose=Component To Local(ComponentPose=@n18)) {cacheName=onlyConstraints}
   n19 = "Constraint - base_ponytail_1 (PPPP)"!(
    ComponentPose=Local To Component(LocalPose=Use cached pose 'startPose'())
    Alpha=1
    ConstraintWeights_0=0.4368
    ConstraintWeights_1:float=0.4024
    ConstraintWeights_2:float=0.0817
    ConstraintWeights_3:float=0.0791
   ) -> (Pose) {
    alphaBoolBlend {blendOption=Linear}
    boneToModify {boneName=base_ponytail_1}
    constraintSetup [{targetBone {boneName=neck_02}}, {targetBone {boneName=neck_01}}, {targetBone {boneName=spine_05}}, {targetBone {boneName=head}}]
    constraintWeights [1,1,1,1]
   }
   Save cached pose 'firstChainConstraint'(Pose=Component To Local(ComponentPose=@n19)) {cacheName=firstChainConstraint}
  ]}
 EventGraph {nodes [
   Event BlueprintUpdateAnimation()
   > n1 = Sequence!(execute) -> (then_0, then_1, then_2:Exec)
   > n2 = Gravity Override State Function DEPRECATED(CharacterPartAnimInstance=Self-Reference(), GameTimeInSeconds=GetGameTimeInSeconds(), GravityOverrideParams=Get Gravity Override Params())
   > Set GravityOverride(GravityOverride=@n2)
   > n3 = Scale Joint Rotation Function!(
    execute
    Scale:Rotator="1,1,0.2"
    TargetBone:Name=head
    ReferenceBone:Name=neck_02
    SkelMeshComponent:Skinned Mesh Component Object Reference=GetPartSkeletalMeshComponent()
   ) -> (then, FinalRotation:Rotator, OriginalRotation:Rotator)
   > Set head rot(head rot:Rotator=@n3) -> (Output_Get:Rotator)
   n4 = BreakRotator(InRot=Get head rot())
   > Set head rot(head rot=MakeRotator(Roll=@n4, Pitch=@n4.Pitch, Yaw=@n4.Yaw))
   n5 = Set Skel Mesh(execute=@n1.then_2, Skel Mesh=GetOwnerSkeletalMeshComponent())
   > Set targeting up(targeting up:double=float * float(
     A=GetAlphaFromReferenceTransformDelta(SkeletalMeshComponent=@n5, BoneName=head, BaseName=neck_01, InputRange=(X=0,Y=-25), TwistAxis="0,0,1")
     B:bool=Get bIsTargeting()
    ) {is=title K2Node_PromotableOperator}) -> (Output_Get:double)
   n6 = Get Skel Mesh()
   > Set Head Scale(Head Scale:Vector=BreakTransform(InTransform=GetBoneTransform!(self:Skinned Mesh Component Object Reference=@n6, InBoneName:Name=head, TransformSpace:ERelativeTransformSpace Enum=RTS_ParentBoneSpace) -> (ReturnValue:Transform)).Scale) -> (Output_Get:Vector)
   n7 = Branch(execute=@n1.then_1, Condition=OR Boolean(A=Get bIsJumping(), B=Get bIsFalling()))
   > Set Sim Space Settings Output(Sim Space Settings Output:Sim Space Settings Structure=Get Sim Space Settings Jump() -> (Sim Space Settings Jump:Sim Space Settings Structure)) -> (Output_Get:Sim Space Settings Structure)
   Set Sim Space Settings Output(execute=@n7.else, Sim Space Settings Output=Get Sim Space Settings Default() -> (Sim Space Settings Default:Sim Space Settings Structure))
   Event OnParachuteOpened()
   > Set alphaHairRot(alphaHairRot:double=1) -> (Output_Get:double)
   > Switch on EGliderType(Selection=Get PlayerGliderType() -> (PlayerGliderType=HangGlider))
   > Set isUsingHangGlidingInParachute(isUsingHangGlidingInParachute:bool=true) -> (Output_Get:bool)
   Event OnParachuteClosed()
   > Set isUsingHangGlidingInParachute(isUsingHangGlidingInParachute=false)
   > Set alphaHairRot(alphaHairRot=0)
  ]}
]
variables [
 alphaHairRot {type=number}
 Gravity Override Params {type=GravityOverrideParamsStruct}
 GravityOverride {type=Vector}
 head rot {type=Rotator}
 Head Scale {type=Vector}
 isUsingHangGlidingInParachute {type=boolean}
 Sim Space Settings Default {
  type=SimSpaceSettings
  default {
   dampingAlpha=1
   maxAngularAcceleration=10000
   maxAngularVelocity=10000
   maxLinearAcceleration=10000
   maxLinearVelocity=10000
   velocityScaleZ=1
   worldAlpha=0
   externalLinearDragV=<0 0 0>
  }
 }
 Sim Space Settings Jump {
  type=SimSpaceSettings
  default {
   dampingAlpha=1
   maxAngularAcceleration=10000
   maxAngularVelocity=10000
   maxLinearAcceleration=10000
   maxLinearVelocity=10000
   velocityScaleZ=1
   worldAlpha=0
   externalLinearDragV=<0 0 0>
  }
 }
 Sim Space Settings Output {
  type=SimSpaceSettings
  default {
   dampingAlpha=1
   maxAngularAcceleration=10000
   maxAngularVelocity=10000
   maxLinearAcceleration=10000
   maxLinearVelocity=10000
   velocityScaleZ=1
   worldAlpha=0
   externalLinearDragV=<0 0 0>
  }
 }
 Skel Mesh {type=SkeletalMeshComponent}
 targeting up {type=number}
]
